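; Gameplay ability for a Battle Pass S28 quest that detects a player landing higher after using a Rift-To-Go.
; It waits for consumed-item events, filters for the Rift tag and records the instigator's starting Z.
; A nested one-shot wait for the landed event records the ending Z.
; If that reaches the start plus an additional required height, it broadcasts the CustomVerb.Land.Higher message.
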(event Ability|EventActivateAbility
  (bind (_asynctaskproxy _payload) (Ability|Async|WaitGameplayEventToActor (Ability|GetAvatarActorfromActorInfo) "(TagName=\"Event.ConsumedItem\")")
    (:EventReceived
      (if (GameplayTags|HasTag (.TargetTags (Utilities|Struct|BreakGameplayEventData _payload)) "(TagName=\"Item.Consumable.Rift\")" true)
        (Variables|Default|SetPlayerStartingAltitude (.z (Transformation|GetActorLocation (.Instigator (Utilities|Struct|BreakGameplayEventData _payload)))))
        (bind (_asynctaskproxy_2 _payload_2) (Ability|Async|WaitGameplayEventToActor (Ability|GetAvatarActorfromActorInfo) "(TagName=\"Event.Physics.Landed\")" true)
          (:EventReceived
            (Variables|Default|SetPlayerEndingAltitude (.z (Transformation|GetActorLocation (.Instigator (Utilities|Struct|BreakGameplayEventData _payload_2)))))
            (if (>= (Variables|Default|GetPlayerEndingAltitude) (+ (Variables|Default|GetPlayerStartingAltitude) (Variables|Default|GetAdditionalRequiredHeight)))
              (CustomVerb|BroadcastCustomVerbMessage (CustomVerb|MakeCustomVerbMessage (Ability|GetAvatarActorfromActorInfo) "(TagName=\"CustomVerb.Land.Higher\")")))))))))

(event Ability|EventOnEndAbility
  (params (bWasCancelled :type Boolean)))
